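# Post-process AnimBlueprint for the ListDizzy_Sashay biped cosmetic companion that drives its tie bones on top of the input pose.
# A weighted Constraint attaches dyn_tie_CST to spine, clavicle, neck and tie bones, then an additive rotation offsets dyn_tie_MSTR.
# AnimDynamics simulates dyn_tie_01 with an angular spring, zero gravity and a spherical head limit.
# The EventGraph remaps head-versus-spine_05 yaw into Head Yaw Back, which the AnimGraph does not read.
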AnimBlueprint /CosmeticCompanions/Assets/Biped/ListDizzy_Sashay/Meshes/ListDizzy_Sashay_Biped_Post_AnimBP
target_skeleton=&/ListDizzy_Sashay_Biped_Skeleton
graphs [
 AnimGraph {nodes [
   n1 = "Constraint - dyn_tie_CST (PPPPPPP)"!(
    ComponentPose=Local To Component(LocalPose=Input Pose())
    Alpha=1
    ConstraintWeights_0=0.0025
    ConstraintWeights_1:float=0.081
    ConstraintWeights_2:float=0.2671
    ConstraintWeights_3:float=0.0353
    ConstraintWeights_4:float=0.023
    ConstraintWeights_5:float=0.0296
    ConstraintWeights_6:float=0.5615
   ) -> (Pose) {
    boneToModify {boneName=dyn_tie_CST}
    constraintSetup [
     {targetBone {boneName=spine_04}}
     {targetBone {boneName=spine_05}}
     {targetBone {boneName=clavicle_l}}
     {targetBone {boneName=clavicle_r}}
     {targetBone {boneName=neck_01}}
     {targetBone {boneName=neck_02}}
     {targetBone {boneName=dyn_tie_01}}
    ]
    constraintWeights [1,1,1,1,1,1,1]
   }
   n2 = "Transform (Modify) Bone - Bone: dyn_tie_MSTR"!(ComponentPose=@n1, Alpha=1, Rotation="-25,-100,0") -> (Pose) {rotationMode=BMM_Additive, rotationSpace=BCS_BoneSpace, boneToModify {boneName=dyn_tie_MSTR}, rotation {pitch=-25, yaw=-100}}
   n3 = Anim Dynamics - Bone: dyn_tie_01!(ComponentPose=@n2, Alpha=1, GravityOverride:Vector="0,0,0") -> (Pose) {
    angularDampingOverride=1
    angularSpringConstant=500
    bAngularSpring=true
    bOverrideAngularDamping=true
    bUseGravityOverride=true
    bUsePlanarLimit=false
    bUseSphericalLimits=true
    simulationSpace=RootRelative
    boundBone {boneName=dyn_tie_01}
    componentAppliedLinearAccClamp=<1000 1000 1000>
    componentLinearAccScale=<0.3 0.3 0.3>
    componentLinearVelScale=<0.3 0.3 0.3>
    physicsBodyDefinitions [{
      collisionType=CustomSphere
      sphereCollisionRadius=2.5
      boundBone {boneName=dyn_tie_01}
      boxExtents=<10 10 10>
      constraintSetup {angularLimitsMax=<0 25 25>, angularLimitsMin=<0 0 -25>, angularTarget=<1 0 0>}
      localJointOffset=<1 0 0>
     }]
    sphericalLimits [{limitRadius=10, drivingBone {boneName=head}, sphereLocalOffset=<10 0 0>}]
   }
   Output Pose(Result=Component To Local(ComponentPose=@n3))
  ]}
 EventGraph {nodes [
   Event BlueprintInitializeAnimation()
   > Set Mesh Component(Mesh Component:Skeletal Mesh Component Object Reference=GetOwningComponent()) -> (Output_Get:Skeletal Mesh Component Object Reference)
   Event BlueprintUpdateAnimation()
   n1 = Get Mesh Component() -> (Mesh Component:Skeletal Mesh Component Object Reference)
   > Set Head Yaw Back(Head Yaw Back:double=MapRangeClamped(
     Value=GetDeltaTransformFromRefPose!(self=@n1, BoneName=head, BaseName=spine_05) -> (
      ReturnValue_Location:Vector
      ReturnValue_Rotation_Roll:float
      ReturnValue_Rotation_Pitch:float
      ReturnValue_Rotation_Yaw:float
      ReturnValue_Scale:Vector
     ).ReturnValue_Rotation_Yaw
     InRangeA=0
     InRangeB=-40
    )) -> (Output_Get:double)
  ]}
]
variables [
 Head Yaw Back {type=number}
 Mesh Component {type=SkeletalMeshComponent}
 PoseAsset {type=PoseAsset, default=/CosmeticCompanions/Automation_Template/companion_pose_asset_template}
]
